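; Vortex area effect actor that grants LuxurySupreme Dudebro exception tags to player pawns inside its geyser cone.
; BeginPlay binds the cone's overlap events, which on the server apply and remove the tag GameplayEffect with debug prints.
; BeginPlay also runs a downward ground trace that stores the hit and moves listed actors there, calling OnGroundLocationEstablished per actor.
; The construction script zeroes the cone's relative rotation.
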(event EventBeginPlay
  (bind _s_fx_geysercone (Variables|Default|GetS_FX_GeyserCone))
  (Collision|BindEventtoOnComponentBeginOverlap _s_fx_geysercone (delegate Custom|OverlapStart))
  (Collision|BindEventtoOnComponentEndOverlap _s_fx_geysercone (delegate Custom|OverlapEnd))
  (CallFunction|DoGroundTrace))

(event Custom|OverlapStart
  (params (OverlappedComponent :type Primitive%20Component%20Object%20Reference) (OtherActor :type Actor%20Object%20Reference) (OtherComp :type Primitive%20Component%20Object%20Reference) (OtherBodyIndex :type Integer) (bFromSweep :type Boolean) (SweepResult :type Hit%20Result%20Structure%20%28by%20ref%29))
  (if (Networking|IsServer)
    (bind _asfort_player_pawn (Utilities|Casting|CastToFortPlayerPawn OtherActor))
    (Fortnite|ApplyGameplayEffectsToActor _asfort_player_pawn (Utilities|Array|MakeArray "/Game/Athena/Environments/Blueprints/DudeBro/GE_LuxurySupreme_DudebroTags.GE_LuxurySupreme_DudebroTags_C") (Transformation|GetActorLocation OtherActor))
    (Development|PrintString "Added exception tags")))

(event Custom|OverlapEnd
  (params (OverlappedComponent :type Primitive%20Component%20Object%20Reference) (OtherActor :type Actor%20Object%20Reference) (OtherComp :type Primitive%20Component%20Object%20Reference) (OtherBodyIndex :type Integer))
  (if (Networking|IsServer)
    (bind _asfort_player_pawn (Utilities|Casting|CastToFortPlayerPawn OtherActor))
    (GameplayEffects|RemoveActiveGameplayEffectbySourceEffect (Ability|GetFortAbilitySystemComponent _asfort_player_pawn) "/Game/Athena/Environments/Blueprints/DudeBro/GE_LuxurySupreme_DudebroTags.GE_LuxurySupreme_DudebroTags_C" 0 -1)
    (Development|PrintString "Removed exception tags" true true "(R=1.000000,G=0.000000,B=0.000000,A=1.000000)")))

(event Custom|Do%20Ground%20Trace
  (bind (_outhit _returnvalue) (Collision|LineTraceByChannel (- (Transformation|GetActorLocation) (Variables|Default|GetTraceStartLocation)) (- (- (Transformation|GetActorLocation) (Variables|Default|GetTraceStartLocation)) (Math|Vector|MakeVector 0.0 0.0 (Variables|Default|GetTraceMaxDistance))) "TraceTypeQuery10"))
  (if _returnvalue
    (Variables|Default|SetTraceHitLocation (.Location (Collision|BreakHitResult _outhit)))
    (Variables|Default|SetTraceHitNormal (.Normal (Collision|BreakHitResult _outhit)))
    (for _array_element (Variables|Default|GetActorsToPlaceAtGround)
      (Transformation|SetActorLocation _array_element (Variables|Default|GetTraceHitLocation) false true)
      (Default|CallOnGroundLocationEstablished))))

(fn ConstructionScript
  (|CallParentFunction)
  (Transformation|SetRelativeRotation (Variables|Default|GetS_FX_GeyserCone) (Math|Rotator|MakeRotator 0.0) false false))
